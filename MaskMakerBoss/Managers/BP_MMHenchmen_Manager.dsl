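; BP_MMHenchmen_Manager is a Mask Maker boss manager that switches the henchmen's Athena patrol paths on or off in response to events.
; On BeginPlay, authority only, it calls the playlist-ready async action and listens on the global event router for MMHenchmen manager events.
; Each event's enum sets bEnablePath, which is then applied to every path in the tunable PatrolPathArray_MMHenchmen.
; The flag starts false but reaches the paths only when an event arrives.
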(event EventBeginPlay
  (Networking|SwitchHasAuthority
    (:Authority
      (Development|PrintString "@@@ Event BeginPlay [BP_PistolCanary_MMHenchmen_Manager]: Authority - CALLED!" false (Variables|Debug|GetbDebugReserve))
      (Class|FortAsyncActionCurrentPlaylistReady|PlaylistDataReadyAsync)
      (bind (_asynctaskproxy _payload) (GameplayEventRouter|StartListeningToEvent (GameplayEventRouter|GetGlobalEventRouter) "/MaskMakerBoss/Managers/MMHenchmen_Manager_Events.MMHenchmen_Manager_Events")
        (:then
          (Variables|Internal|SetAsyncTaskEventListener _asynctaskproxy)
          (Variables|Internal|SetbEnablePath false))
        (:OnEventReceived
          (switch Utilities|FlowControl|Switch|SwitchonMMHenchmen_Events_Enum (Utilities|Struct|BreakMMHenchmenManagerEvents _payload)
            (:NewEnumerator0
              (Variables|Internal|SetbEnablePath true))
            (:NewEnumerator1
              (Variables|Internal|SetbEnablePath false)))
          (for _array_element (Variables|Tunable|GetPatrolPathArray_MMHenchmen)
            (Class|FortAthenaPatrolPath|SetPatrolPathEnabled _array_element (Variables|Internal|GetbEnablePath))))))))

(fn ConstructionScript)
